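# Parent material for Nirvana vehicle wheels, with animated, tintable glow on rim and tire.
# Vertex color red splits rim from tire: the rim gets panning arcs masked by its RGB texture, while the tire layers tread masks, fresnel, scrolling gradients, noise and twinkles.
# CustomColor tints the glow, and Over_Punch adds white boost, except one fresnel layer reading separate "Over Punch".
# Shared damage and de-rez functions wrap the output attributes.
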Material /VehicleCosmetics/Wheels/Nirvana/Materials/M_WheelParent_Nirvana
graph {
 properties {bAutomaticallySetUsageInEditor=false, bUseMaterialAttributes=true, bUsedWithSkeletalMesh=true, usedShadingModels=Default Lit}
 nodes [
  n1 = ScalarParameter() {defaultValue=1, parameterName=Bloom_Speed}
  n2 = Time()
  n3 = TextureSampleParameter2D(UVs=Panner(Coordinate=TextureCoordinate() {vTiling=6}, Time=Multiply(A=@n1, B=@n2)) {speedY=-0.5}) {parameterName=Arcs, samplerType=SAMPLERTYPE_LinearColor, texture=/VehicleCosmetics/Wheels/_SharedTextures/FX/arcs}
  n4 = ConstantBiasScale(None=@n3.G) {bias=-0.8, scale=3}
  n5 = ConstantBiasScale(None=@n3.G) {bias=-0.2, scale=0.2}
  n6 = TextureSampleParameter2D(UVs=TextureCoordinate()) {
   parameterName=Arcs
   samplerSource=SSM_Wrap_WorldGroupSettings
   samplerType=SAMPLERTYPE_LinearColor
   texture=&arcs
  }
  n7 = TextureSampleParameter2D() {parameterName=Rim_RGB, samplerType=SAMPLERTYPE_LinearColor, texture=/VehicleCosmetics/Wheels/Nirvana/Textures/T_Nirvana_Rim_RGB}
  n8 = Multiply(
   A=Add(A=Clamp(None=@n4) {maxDefault=1.2}, B=Clamp(None=@n5))
   B=Multiply(A=ConstantBiasScale(None=@n6.G) {bias=0, scale=1}, B=OneMinus(None=@n7.B))
  )
  n9 = Multiply(A=Constant() {r=3}, B=Add(A=@n8, B=Multiply(A=@n8, B=@n7.R)))
  n10 = Multiply(A=VectorParameter() {parameterName=CustomColor, defaultValue=<rgba 0.428 0 0.481 1>}.RGB, B=@n9)
  n11 = ScalarParameter() {defaultValue=0.14, parameterName=Over_Punch}
  n12 = Clamp(None=ConstantBiasScale(None=Multiply(A=@n9, B=@n11)) {bias=-0.1, scale=1})
  n13 = Time()
  n14 = Constant() {r=1.8}
  n15 = TextureSampleParameter2D(UVs=Panner(Coordinate=TextureCoordinate(), Time=Add(A=@n13, B=@n14)) {speedY=0.083}) {parameterName=Arcs, samplerType=SAMPLERTYPE_LinearColor, texture=&arcs}
  n16 = Max(A=@n15.R, B=@n15.R)
  n17 = Constant() {r=2}
  n18 = Constant2Vector() {g=0.4, r=1}
  n19 = Add(A=TextureCoordinate() {vTiling=0.62}, B=AppendVector(A=Constant(), B=Constant() {r=0.35}))
  n20 = TextureSample(UVs=Rotator(Coordinate=Add(A=@n18, B=@n19), Time=Constant() {r=1.5}) {speed=3.1415}) {texture=/VehicleCosmetics/Boosts/Boost_Fireball/Textures/T_VCVFX_Boost_Fireball_Gradiante_1_M}
  n21 = Multiply(
   A=OneMinus(None=MaterialFunctionCall!(Normal, Exponent=@n17) {materialFunction=/VehicleCosmetics/Wheels/Cola_Holographic/Materials/MF_Fresnel_Classic}.Result)
   B=ConstantBiasScale(None=@n20.R) {bias=0, scale=1.5}
  )
  n22 = ConstantBiasScale(None=Multiply(A=ScalarParameter() {parameterName=Over_Punch}, B=Multiply(A=@n16, B=@n21))) {bias=-0.02, scale=3}
  n23 = Max(
   A=Multiply(A=VectorParameter() {parameterName=CustomColor, defaultValue=<rgba 0.428 0 0.481 1>}.RGB, B=@n15.R)
   B=Multiply(A=VectorParameter() {parameterName=LightAccentColor, defaultValue=<rgba 0.923 0.067 0.178 1>}.RGB, B=@n15.B)
  )
  n24 = TextureSampleParameter2D() {parameterName=Tread_RGB, samplerType=SAMPLERTYPE_LinearColor, texture=/VehicleCosmetics/Wheels/Nirvana/Textures/T_Nirvana_Tire_RGB}
  n25 = ConstantBiasScale(None=@n24.R) {bias=0, scale=4}
  n26 = ConstantBiasScale(None=@n24.B) {bias=0, scale=4}
  n27 = Multiply(A=Add(A=Clamp(None=@n22), B=Multiply(A=@n23, B=@n21)), B=Clamp(None=Max(A=@n25, B=@n26)))
  n28 = TextureSampleParameter2D() {parameterName=Tread_RGB, samplerType=SAMPLERTYPE_LinearColor, texture=&T_Nirvana_Tire_RGB}
  n29 = ConstantBiasScale(None=@n28.B) {bias=0.66, scale=1}
  n30 = MaterialFunctionCall() {materialFunction=&MF_Fresnel_Classic}
  n31 = Subtract(A=OneMinus(None=@n30.Result), B=Constant() {r=0.6})
  n32 = Constant2Vector() {g=0.5, r=1}
  n33 = TextureCoordinate() {coordinateIndex=2}
  n34 = TextureSample(UVs=Panner(Coordinate=Multiply(A=@n32, B=@n33)) {speedX=0.02, speedY=0.03}) {texture=/VehicleCosmetics/Textures/UniversalTextures/Universal_Noise_55_RGB}
  n35 = Multiply(
   A=@n34.G
   B=TextureSample(UVs=Panner(Coordinate=@n33) {speedX=0.03, speedY=0.05}) {texture=&Universal_Noise_55_RGB}.G
  )
  n36 = ConstantBiasScale(None=Multiply(A=OneMinus(None=@n20.R), B=@n35)) {bias=0.2, scale=0.6}
  n37 = Constant() {r=0.4}
  n38 = Multiply(A=ScalarParameter() {defaultValue=0.14, parameterName=Over Punch}, B=Max(A=Multiply(A=@n31, B=@n36), B=Multiply(A=@n30.Result, B=@n37)))
  n39 = ConstantBiasScale(None=@n38) {bias=-0.01, scale=1.5}
  n40 = Multiply(A=Add(A=Clamp(None=@n39), B=@n38), B=VectorParameter() {parameterName=CustomColor, defaultValue=<rgba 0.428 0 0.481 1>}.RGB)
  n41 = ScalarParameter() {parameterName=Over_Punch}
  n42 = TextureSampleParameter2D() {parameterName=Tread_RGB, samplerType=SAMPLERTYPE_LinearColor, texture=&T_Nirvana_Tire_RGB}
  n43 = ScalarParameter() {defaultValue=1, parameterName=Bloom_Speed}
  n44 = Add(A=Constant() {r=0.4}, B=Time())
  n45 = Rotator(Coordinate=Panner(Coordinate=TextureCoordinate() {vTiling=2}, Time=Multiply(A=@n43, B=@n44)) {speedY=0.5}, Time=Constant() {r=1}) {speed=3.1415}
  n46 = ConstantBiasScale(None=Multiply(A=@n20.R, B=TextureSample(UVs=@n45) {texture=/VehicleCosmetics/Wheels/_SharedTextures/FX/Gradient_V_D}.R)) {bias=0, scale=3}
  n47 = Clamp(None=ConstantBiasScale(None=@n20.R) {bias=-0.4, scale=2.5})
  n48 = ConstantBiasScale(None=@n35) {bias=0, scale=1.2}
  n49 = Max(A=Multiply(A=ConstantBiasScale(None=@n42.R) {bias=0, scale=1}, B=@n46), B=Multiply(A=@n46, B=Max(A=@n47, B=@n48)))
  n50 = Clamp(None=ConstantBiasScale(None=Multiply(A=@n41, B=@n49)) {bias=-0.02, scale=2})
  n51 = Multiply(A=@n49, B=VectorParameter() {parameterName=CustomColor, defaultValue=<rgba 0.428 0 0.481 1>}.RGB)
  n52 = Constant() {r=2}
  n53 = Multiply(A=TextureCoordinate() {coordinateIndex=2, unMirrorV=true}, B=Constant2Vector() {g=3, r=2})
  n54 = OneMinus(None=@n35)
  n55 = Multiply(
   A=TextureSample(UVs=Panner(Coordinate=@n53) {speedX=0.25, speedY=0.12}) {texture=/VehicleCosmetics/Wheels/_SharedTextures/FX/skin_twankle2_rgb}.RGB
   B=Clamp(None=ConstantBiasScale(None=@n54) {bias=-0.75, scale=2})
  )
  n56 = Constant() {r=1.5}
  n57 = Add(A=Add(A=Multiply(A=@n29, B=@n40), B=Add(A=@n50, B=@n51)), B=Max(A=Power(Base=@n34.R, Exp=@n52), B=Power(Base=@n55, Exp=@n56)))
  n58 = OneMinus(None=VertexColor().R)
  n59 = Multiply(A=Constant() {r=0.2}, B=@n24.R)
  n60 = SetMaterialAttributes!(
   MaterialAttributes
   Base Color=VectorParameter() {parameterName=BaseColor}.RGBA
   Emissive Color=LinearInterpolate(A=Add(A=@n10, B=@n12), B=Max(A=@n27, B=@n57), Alpha=VertexColor().R)
   Metallic=ScalarParameter() {parameterName=Metallic}
   Roughness=LinearInterpolate(
    A=ScalarParameter() {defaultValue=0.4, parameterName=Tire_Roughness}
    B=ScalarParameter() {defaultValue=0.9, parameterName=Rim_Roughness}
    Alpha=Add(A=@n58, B=@n59)
   )
  ) {attributeSetTypes [
    #1
    #2
    #3
    #4
   ]}
  MaterialFunctionCall(In=MaterialFunctionCall(Material=@n60) {materialFunction=&MF_BaseVehicle_Damage}.Result) {materialFunction=&MF_VehicleDeRez_Transition}
 ]
}
